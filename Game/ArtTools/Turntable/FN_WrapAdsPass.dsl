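; FN_WrapAdsPass is a Blueprint in the ArtTools Turntable folder that links itself to its owning turntable.
; On BeginPlay it calls GetOwningTurntable and casts the result to FN_WrapTurntable.
; The graph shown does nothing further with the cast result, so the pass behaviour itself is not visible here.
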(event EventBeginPlay
  (bind _asfn_wrap_turntable (Utilities|Casting|CastToFN_WrapTurntable (CallFunction|GetOwningTurntable))))
