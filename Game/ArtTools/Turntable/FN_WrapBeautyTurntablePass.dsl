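; Turntable beauty pass that applies a chosen wrap to every FortWrapSkeletalMeshActor, then captures screenshots through the owning turntable.
; Prepare also caches every directional light's rotation, which cleanup restores.
; With turnObjectNotCamera set, each auto capture lerps a yaw around 360 degrees at fixed pitch, turning the lights with the camera rather than moving any meshes.
; Placed-camera captures and the normal mode defer to the parent class.
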(event Camera|EventcaptureNextCamera
  (if (Variables|Default|GetturnObjectNotCamera)
    (if (> (Utilities|Array|Length (Variables|Camera|GetcurrentCameras)) 0)
      (if (Variables|Turntable|GetusePlacedAndAutoCameras)
        (if (Utilities|Array|IsValidIndex (Variables|Camera|GetcurrentCameras) (+ (Variables|Camera|GetcurrentCameraIndex) 1))
          (|CallParentFunction)
          (else
            _))
        (else
          (|CallParentFunction))))
    (Math|Integer|IncrementInt (Variables|Camera|GetcurrentCameraIndex)
      (:%20%20
        (CallFunction|TurnObjectsandCapture)))
    (else
      (|CallParentFunction))))

(fn cleanUpPass
  (Utilities|FlowControl|Sequence
    (:then_0
      (bind _keys (Utilities|Map|Keys (Variables|Default|GetdirectionalLightMap)))
      (for _array_element _keys
        (Transformation|SetWorldRotation _array_element (Utilities|Map|Find (Variables|Default|GetdirectionalLightMap) _array_element) false false)))
    (:then_1
      (bind _tmp (|CallParentFunction))
      (return _tmp))))

(fn getAutoTurntableRotationFromIndex
  (params (currentCamerIndex :type Integer) (TotalCameraCount :type Integer))
  (return (Math|Rotator|Lerp(Rotator) "-20.000000,0.000000, 0" "-20.000000,360.000000,0.000000" (/ (Math|Conversions|ToFloat(Integer) currentCamerIndex) (Math|Conversions|ToFloat(Integer) TotalCameraCount)))))

(fn initDirectionalLights
  (Utilities|Map|Clear (Variables|Default|GetdirectionalLightMap))
  (bind _outactors (Actor|GetAllActorsOfClass "/Script/Engine.Actor"))
  (for _array_element _outactors
    (Utilities|IsValid (Actor|GetComponentbyClass _array_element "/Script/Engine.DirectionalLightComponent")
      (:Is%20Valid
        (Utilities|Map|Add (Variables|Default|GetdirectionalLightMap) (Actor|GetComponentbyClass _array_element "/Script/Engine.DirectionalLightComponent") (Transformation|GetWorldRotation (Actor|GetComponentbyClass _array_element "/Script/Engine.DirectionalLightComponent")))))))

(fn preparePass
  (|CallParentFunction)
  (Utilities|FlowControl|Sequence
    (:then_0
      (bind _outactors (Actor|GetAllActorsOfClass "/Script/FortniteGame.FortWrapSkeletalMeshActor"))
      (for _array_element _outactors
        (Class|FortWrapSkeletalMeshActor|SetWrap _array_element (Variables|Default|GetwrapToUse) (Variables|Default|GetwrapType))))
    (:then_1
      (CallFunction|InitDirectionalLights))
    (:then_2
      (return false))))

(fn turnObjectsAndCapture
  (bind (_l_rot_roll _l_rot_pitch _l_rot_yaw) (Math|Rotator|BreakRotator (Variables|Getl_rot_Roll)))
  (bind _returnvalue (CallFunction|GetAutoTurntableRotationfromIndex (Variables|Camera|GetcurrentCameraIndex) (Variables|Turntable|GetautoTurntableCameraCount)))
  (LocalVariables|Setl_rot _returnvalue)
  (Utilities|FlowControl|Sequence
    (:then_0
      (bind _keys (Utilities|Map|Keys (Variables|Default|GetdirectionalLightMap)))
      (for _array_element _keys
        (Transformation|SetWorldRotation _array_element (Math|Rotator|MakeRotator (.roll (Utilities|Map|Find (Variables|Default|GetdirectionalLightMap) _array_element)) (.pitch (Utilities|Map|Find (Variables|Default|GetdirectionalLightMap) _array_element)) _l_rot_yaw) false false)))
    (:then_1
      (if (< (Variables|Camera|GetcurrentCameraIndex) (+ (Variables|Turntable|GetautoTurntableCameraCount) (Utilities|Array|Length (Variables|Camera|GetcurrentCameras))))
        (Class|MultiPassTTEditorInterface|TakeScreenshotfromTransform(Message) (CallFunction|GetOwningTurntable) "0, 0, 0" (LocalVariables|Getl_rot) "0, 0, 0" 0.0 self)
        (else
          (CallFunction|EndPass))))))
