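; Turntable capture pass that renders a wrap cosmetic on the wrapping paper roll mesh.
; On BeginPlay it runs the parent pass only when the current capture actor is a wrap skeletal mesh actor using that roll mesh, and otherwise ends the pass.
; Pass preparation applies the configured wrap and type to every wrap mesh actor in the level.
; Auto rotation spreads camera indices over a full yaw turn at a fixed downward pitch.
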(event EventBeginPlay
  (bind _asfn_wrap_turntable (Utilities|Casting|CastToFN_WrapTurntable (CallFunction|GetOwningTurntable))
    (:then
      (bind _asfort_wrap_skeletal_mesh_actor (Utilities|Casting|CastToFortWrapSkeletalMeshActor (Variables|/Game/FN_PublicTurntable/MultipassTurntable/MultipassTurntableActor_MultiCapture.MultipassTurntableActor_MultiCapture_C|GetcurrentCaptureActor _asfn_wrap_turntable))
        (:then
          (if (Utilities|Equal(Object) (Variables|/Script/Engine.SkinnedMeshComponent|GetSkeletalMesh (Variables|/Script/Engine.SkeletalMeshActor|GetSkeletalMeshComponent _asfort_wrap_skeletal_mesh_actor)) "/Game/UI/Wrapping_Paper/Meshes/Wrapping_Papper_Aligned.Wrapping_Papper_Aligned")
            (|CallParentFunction)
            (else
              _)))
        (:CastFailed
          _)))
    (:CastFailed
      _))
  (CallFunction|EndPass))

(fn getAutoTurntableRotationFromIndex
  (params (currentCamerIndex :type Integer) (TotalCameraCount :type Integer))
  (return (Math|Rotator|Lerp(Rotator) "-20.000000,0.000000, 0" "-20.000000,360.000000,0.000000" (/ (Math|Conversions|ToFloat(Integer) currentCamerIndex) (Math|Conversions|ToFloat(Integer) TotalCameraCount)))))

(fn preparePass
  (|CallParentFunction)
  (bind _outactors (Actor|GetAllActorsOfClass "/Script/FortniteGame.FortWrapSkeletalMeshActor"))
  (for _array_element _outactors
    (Class|FortWrapSkeletalMeshActor|SetWrap _array_element (Variables|Default|GetwrapToUse) (Variables|Default|GetwrapType)))
  (return false))
