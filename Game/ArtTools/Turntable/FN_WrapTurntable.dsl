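; Fortnite art-tool turntable, extending MultipassTurntable_MultiCapture, that captures every selected wrap across a list of meshes, then exits.
; Each capture reuses a name-matched weapon or vehicle override actor or spawns a FortWrapSkeletalMeshActor, then applies mesh, animation and transform.
; It passes the wrap and name-derived wrap type to the beauty pass and shows the wrap name on the HUD.
; An editor bridge component supplies the wrap selection.
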(event EventBeginPlay
  (Development|ExecuteConsoleCommand "r.SetNearClipPlane 5")
  (Variables|/Game/FN_PublicTurntable/MultipassTurntable/ScreenshotPasses/TurntablePass_Beauty.TurntablePass_Beauty_C|SetautoTurntableCameraCount (Variables|WrapTurntable|GetnumberOfCapturePerMesh) (Variables|Default|GetFN_WrapBeautyTurntablePass))
  (Utilities|FlowControl|Delay (Variables|Turntable|GetinitialStartDelay))
  (Variables|Turntable|SetinitialStartDelay 0.0)
  (|CallParentFunction))

(event MultipassTurntable_MultiCapture|EventspawnNewCaptureActor
  (CallFunction|SpawnWrapActor))

(event MultipassTurntable_MultiCapture|EventhandleCaptureComplete
  (params (actor :type Multipass%20Turntable%20Actor%20Object%20Reference))
  (if (Utilities|Array|IsValidIndex (Variables|MultipassTurntable_MultiCapture|GetMultiCaptureActorClasses) (+ (Variables|MultipassTurntable_MultiCapture|GetcurrentCaptureIndex) 1))
    (|CallParentFunction 0)
    (else
      (Math|Integer|IncrementInt (Variables|Default|GetcurrentWrapIndex)
        (:%20%20
          (if (Utilities|Array|IsValidIndex (Variables|WrapTurntable|GetwrapsToCapture) (Variables|Default|GetcurrentWrapIndex))
            (Variables|MultipassTurntable_MultiCapture|SetcurrentCaptureIndex -1)
            (Captures|beginCapture 0.05)
            (else
              (Default|ExitGame))))))))

(event MultipassTurntable_MultiCapture|EventsetupNextCapture
  (bind _returnvalue (CallFunction|GetValidWrapActorOverride (+ (Variables|MultipassTurntable_MultiCapture|GetcurrentCaptureIndex) 1)))
  (Utilities|IsValid _returnvalue
    (:Is%20Valid
      (Math|Integer|IncrementInt (Variables|MultipassTurntable_MultiCapture|GetcurrentCaptureIndex)
        (:%20%20
          (Utilities|FlowControl|Sequence
            (:then_0
              (Variables|MultipassTurntable_MultiCapture|SetcurrentCaptureActor _returnvalue)
              (Rendering|SetActorHiddenInGame (Variables|Default|GetwrapActorOverride_Weapon) true)
              (Rendering|SetActorHiddenInGame (Variables|MultipassTurntable_MultiCapture|GetcurrentCaptureActor) false))
            (:then_1
              (if (Utilities|Array|IsValidIndex (Variables|MultipassTurntable_MultiCapture|GetMultiCaptureActorClasses) (Variables|MultipassTurntable_MultiCapture|GetcurrentCaptureIndex))
                (Utilities|IsValid (Variables|MultipassTurntable_MultiCapture|GetcurrentCaptureActor)
                  (:Is%20Valid
                    _)
                  (:Is%20Not%20Valid
                    (MultipassTurntable_MultiCapture|spawnNewCaptureActor)))
                (CallFunction|FinalizeWrapActor)
                (else
                  (Default|ExitGame))))))))
    (:Is%20Not%20Valid
      (|CallParentFunction))))

(event Camera|EventinitializeShotCamera
  (|CallParentFunction)
  (Camera|SetFieldOfView (Variables|/Script/Engine.CameraActor|GetCameraComponent (Variables|Camera|GetcurrentCameraActor)) 3e+01))

(event Captures|EventbeginCapture
  (params (delay :type Float%20%28double-precision%29))
  (|CallParentFunction 0.0))

(fn finalizeWrapActor
  (bind (_output_optionalname_2_8af43b6847a8df0006c524bd766166b3 _output_asset_18_e92d0c36446b2cf7f5dea4a06224bead _output_transform_14_bde80147472a0438a675e691fde2d4de) (Utilities|Struct|BreakMultipassTTSpawnStruct (Utilities|Array|Get(acopy) (Variables|MultipassTurntable_MultiCapture|GetMultiCaptureActorClasses) (Variables|MultipassTurntable_MultiCapture|GetcurrentCaptureIndex))))
  (bind _currentcaptureindex (Variables|MultipassTurntable_MultiCapture|GetcurrentCaptureIndex))
  (bind _multicaptureactorclasses (Variables|MultipassTurntable_MultiCapture|GetMultiCaptureActorClasses))
  (bind (_output_optionalname_2_8af43b6847a8df0006c524bd766166b3_2 _output_asset_18_e92d0c36446b2cf7f5dea4a06224bead_2 _output_transform_14_bde80147472a0438a675e691fde2d4de_2 _output_animation_17_878352ff4759a636dc5481bc4f7aa50b) (Utilities|Struct|BreakMultipassTTSpawnStruct (Utilities|Array|Get(acopy) _multicaptureactorclasses _currentcaptureindex)))
  (Utilities|FlowControl|Sequence
    (:then_0
      (bind _asfort_wrap_skeletal_mesh_actor (Utilities|Casting|CastToFortWrapSkeletalMeshActor (Variables|MultipassTurntable_MultiCapture|GetcurrentCaptureActor)))
      (bind _asskeletal_mesh (Utilities|Casting|CastToSkeletalMesh _output_asset_18_e92d0c36446b2cf7f5dea4a06224bead_2))
      (Components|SkinnedMesh|SetSkinnedAssetandUpdate (Variables|/Script/Engine.SkeletalMeshActor|GetSkeletalMeshComponent _asfort_wrap_skeletal_mesh_actor) _asskeletal_mesh true)
      (Utilities|IsValid _output_animation_17_878352ff4759a636dc5481bc4f7aa50b
        (:Is%20Valid
          (Components|Animation|SetAnimation (Variables|/Script/Engine.SkeletalMeshActor|GetSkeletalMeshComponent _asfort_wrap_skeletal_mesh_actor) _output_animation_17_878352ff4759a636dc5481bc4f7aa50b))))
    (:then_1
      (bind _type (CallFunction|GetCurrentWrapType (Variables|Default|GetcurrentWrapIndex)))
      (Variables|/Game/ArtTools/Turntable/FN_WrapBeautyTurntablePass.FN_WrapBeautyTurntablePass_C|SetwrapType _type (Variables|Default|GetFN_WrapBeautyTurntablePass))
      (Variables|/Game/ArtTools/Turntable/FN_WrapBeautyTurntablePass.FN_WrapBeautyTurntablePass_C|SetwrapToUse (Utilities|Array|Get(acopy) (Variables|WrapTurntable|GetwrapsToCapture) (Variables|Default|GetcurrentWrapIndex)) (Variables|Default|GetFN_WrapBeautyTurntablePass))
      (CallFunction|UpdateHUDText))
    (:then_2
      (Transformation|SetActorTransform (Variables|MultipassTurntable_MultiCapture|GetcurrentCaptureActor) (Math|Transform|MakeRelativeTransform _output_transform_14_bde80147472a0438a675e691fde2d4de (Variables|Turntable|GetSpawnedActorTransformOffset)) false false))))

(fn getBoundsOfRenderActors
  (params (offset :type Vector :default "0, 0, 0") (minRadius :type Float%20%28double-precision%29))
  (bind (_boundscenter _radius _boxextent) (|CallParentFunction offset minRadius))
  (return _boundscenter (* _radius 0.75) _boxextent))

(fn getCurrentWrapType
  (params (captureIndex :type Integer))
  (bind _output_optionalname_2_8af43b6847a8df0006c524bd766166b3 (Utilities|Struct|BreakMultipassTTSpawnStruct (Utilities|Array|Get(acopy) (Variables|MultipassTurntable_MultiCapture|GetMultiCaptureActorClasses) captureIndex)))
  (return (select (Utilities|String|Contains (Utilities|String|ToString(Name) _output_optionalname_2_8af43b6847a8df0006c524bd766166b3) "veh") "VehicleWrap_Opaque" "WeaponWrap")))

(fn getValidWrapActorOverride
  (params (inCaptureIndex :type Integer))
  (bind _incaptureindex (Variables|GetinCaptureIndex))
  (bind _multicaptureactorclasses (Variables|MultipassTurntable_MultiCapture|GetMultiCaptureActorClasses))
  (bind _output_optionalname_2_8af43b6847a8df0006c524bd766166b3 (Utilities|Struct|BreakMultipassTTSpawnStruct (Utilities|Array|Get(acopy) _multicaptureactorclasses _incaptureindex)))
  (bind _returnvalue (Utilities|String|ToString(Name) _output_optionalname_2_8af43b6847a8df0006c524bd766166b3))
  (Utilities|FlowControl|Sequence
    (:then_0
      (if (Utilities|String|Contains _returnvalue "weap")
        (LocalVariables|Setl_currentWrapActor (Variables|Default|GetwrapActorOverride_Weapon))
        (elif (Utilities|String|Contains _returnvalue "veh")
          (if (Utilities|String|Contains _returnvalue "mask")
            (LocalVariables|Setl_currentWrapActor 0)
            (else
              (LocalVariables|Setl_currentWrapActor (Variables|Default|GetwrapActorOverride_VehicleOpaque))))))
      (Development|PrintString (Utilities|String|Append (Utilities|String|ToString(Integer) (Variables|GetinCaptureIndex)) " chose " (Utilities|GetDisplayName (LocalVariables|Getl_currentWrapActor))))
      (return (LocalVariables|Getl_currentWrapActor)))
    (:then_1
      (bind _output_get (LocalVariables|Setl_currentWrapActor (Variables|Default|GetwrapActorOverride_Weapon)))
      (Utilities|IsValid _output_get
        (:Is%20Valid
          (return (LocalVariables|Getl_currentWrapActor)))))
    (:then_2
      (bind _output_get_2 (LocalVariables|Setl_currentWrapActor (Variables|Default|GetwrapActorOverride_VehicleOpaque)))
      (Utilities|IsValid _output_get_2
        (:Is%20Valid
          (return (LocalVariables|Getl_currentWrapActor)))))
    (:then_3
      (bind _output_get_3 (LocalVariables|Setl_currentWrapActor 0))
      (Utilities|IsValid _output_get_3
        (:Is%20Valid
          (return (LocalVariables|Getl_currentWrapActor)))))))

(fn gotoWrapsFolder
  (Class|FNWrapTurntableEditorBridgeInterface|GotoWrapsFolder(Message) (Variables|Default|GetFN_WrapTurntable_EditorComp)))

(fn loadSelectedWraps
  (bind _wraps (Class|FNWrapTurntableEditorBridgeInterface|LoadSelectedWraps(Message) (Variables|Default|GetFN_WrapTurntable_EditorComp)))
  (Variables|WrapTurntable|SetwrapsToCapture _wraps))

(fn spawnWrapActor
  (bind (_output_optionalname_2_8af43b6847a8df0006c524bd766166b3 _output_asset_18_e92d0c36446b2cf7f5dea4a06224bead _output_transform_14_bde80147472a0438a675e691fde2d4de) (Utilities|Struct|BreakMultipassTTSpawnStruct (Utilities|Array|Get(acopy) (Variables|MultipassTurntable_MultiCapture|GetMultiCaptureActorClasses) (Variables|MultipassTurntable_MultiCapture|GetcurrentCaptureIndex))))
  (bind _returnvalue (Game|SpawnActorfromClass "/Script/FortniteGame.FortWrapSkeletalMeshActor" (Math|Transform|MakeRelativeTransform (Variables|Turntable|GetSpawnedActorTransformOffset) _output_transform_14_bde80147472a0438a675e691fde2d4de) "Undefined" "MultiplyWithRoot" 0 0))
  (Variables|MultipassTurntable_MultiCapture|SetcurrentCaptureActor _returnvalue)
  (CallFunction|FinalizeWrapActor))

(fn tryInitialize
  (Utilities|IsValid (Game|GetPlayerController 0)
    (:Is%20Valid
      (|CallParentFunction))
    (:Is%20Not%20Valid
      (Utilities|Time|SetTimerbyEvent (EventDispatchers|CreateEvent :!event tryInitialize) 0.1))))

(fn updateHUDText
  (bind _returnvalue (Game|GetPlayerController 0))
  (Utilities|IsValid _returnvalue
    (:Is%20Valid
      (bind _asmultipass_tt_hud (Utilities|Casting|CastToMultipassTT_HUD (HUD|GetHUD _returnvalue)))
      (Class|MultipassTTHUD|SetScreenText _asmultipass_tt_hud (Utilities|GetObjectName (Utilities|Array|Get(acopy) (Variables|WrapTurntable|GetwrapsToCapture) (Variables|Default|GetcurrentWrapIndex))) "(R=1.000000,G=0.239216,B=0.039216,A=1.000000)" 1.5 "/FortUILibrary/Fonts/Burbank.Burbank" :!class /Game/FN_PublicTurntable/MultipassTurntable/Game/MultipassTT_HUD.MultipassTT_HUD_C))))

(fn ConstructionScript
  (|CallParentFunction))
