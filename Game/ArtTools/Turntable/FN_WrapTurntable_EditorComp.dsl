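; Editor-side helper Blueprint for the item wrap turntable art tool, judging by its name and path.
; gotoWrapsFolder syncs the Content Browser to a reference wrap asset, Wrap_001_ArcticCamo, to reveal the ItemWraps folder.
; loadSelectedWraps takes the assets selected in the editor and returns only those of class AthenaItemWrapDefinition.
; BeginPlay is left empty.
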(event EventBeginPlay)

(fn gotoWrapsFolder
  (EditorScripting|ContentBrowser|SyncBrowsertoObjects (Utilities|Array|MakeArray "/Game/Athena/Items/Cosmetics/ItemWraps/Wrap_001_ArcticCamo"))
  (return false))

(fn loadSelectedWraps
  (bind _returnvalue (Development|Editor|GetSelectedAssets))
  (bind _returnvalue_2 (EditorScripting|Utilities|Filter|FilterbyClass _returnvalue "/Script/FortniteGame.AthenaItemWrapDefinition"))
  (return _returnvalue_2))
